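# Landscape material function for Apollo roads: blends painted-line asphalt attributes with an incoming aggregate material and outputs masks, crack height and the asphalt material.
# Asphalt comes from MF_ApolloAsphalt or a Helios layer match by static switch; packed mask and normal textures add line color, roughness, specular and normals.
# Forward and mobile paths use vertex normals, constant line roughness and an optional aggregate blend.
# The gradient road-tint branch feeds no output.
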MaterialFunction /Game/Athena/Apollo/Environments/Landscape/Materials/MaterialFunctions/MF_ApolloRoad
graph {
 properties {libraryCategoriesText []}
 nodes [
  n1 = BlendMaterialAttributes(
   A=FunctionInput() {description=Aggregate Input Material, inputName=Aggregate In, inputType=FunctionInput_MaterialAttributes}
   B=NamedRerouteUsage()
   Alpha=NamedRerouteUsage()
  ) {desc="Vertex blend type altered!", vertexAttributeBlendType=UseA}
  n2 = StaticSwitchParameter(True=@n1, False=NamedRerouteUsage()) {defaultValue=true, group=Mobile, parameterName=Mobile Aggregate}
  FunctionOutput(None=ShadingPathSwitch(Default=@n1, Forward=@n2, Mobile=@n2))
  n3 = NamedRerouteUsage()
  n4 = Saturate(None=Multiply(A=Subtract(A=@n3) {constB=0.1}) {constB=3})
  FunctionOutput(None=NamedRerouteDeclaration(None=@n4) {name=Mask, nodeColor=<rgba 0.5469 0 1 1>}) {outputName=Alpha, sortPriority=1}
  n5 = NamedRerouteUsage()
  FunctionOutput(None=LinearInterpolate(A=Ceil(None=@n5), B=@n5, Alpha=@n4)) {outputName=Tire_Alpha, sortPriority=1}
  n6 = MaterialFunctionCall!(TilingSize) {materialFunction=&/MF_ApolloAsphalt}
  NamedRerouteDeclaration(None=FunctionOutput(None=@n6.SimpleCrackHeight) {outputName=AsphaltCrackHeight, sortPriority=5}) {name=AsphaltCrack, nodeColor=<rgba 0.5469 0 1 1>}
  n7 = MaterialFunctionCall!(Grass_UseCustomSlopeMask, Grass_CustomSlopeMask) {materialFunction=/Game/Athena/Helios/Environments/Landscape/Materials/MaterialFunctions_Utilities/MF_Helios_StaticMeshLandscapeLayerMatch}
  NamedRerouteDeclaration(None=FunctionOutput(None=StaticSwitchParameter(True=@n7.AsphaltLayer, False=@n6.Result) {parameterName=UseHeliosAsphalt}) {outputName=AsphaltMaterial, sortPriority=4}) {name=AsphaltMaterial, nodeColor=<rgba 0.5469 0 1 1>}
  n8 = TextureSampleParameter2D!(UVs, Bias=Constant() {r=-0.5}, Apply View MipBias) {
   group=Road Lines
   mipValueMode=TMVM_MipBias
   parameterName=Road Line Mask
   samplerType=SAMPLERTYPE_Masks
   texture=/Game/Athena/Apollo/Environments/Landscape/Textures/Roads/AsphaltRoad1_rgb
  }
  NamedRerouteDeclaration(None=LinearInterpolate(
    A=VectorParameter() {group=Road Lines, parameterName=EdgeLine, defaultValue=<rgba 0.7344 0.6848 0.6196 0>}.RGB
    B=VectorParameter() {group=Road Lines, parameterName=CenterLine, defaultValue=<rgba 0.4125 0.2346 0 1>}.RGB
    Alpha=@n8.B
   )) {name=LineColor, nodeColor=<rgba 0.5469 0 1 1>}
  n9 = NamedRerouteUsage()
  NamedRerouteDeclaration(None=Multiply(A=OneMinus(None=@n9), B=@n8.G)) {name=RoadLineBlend, nodeColor=<rgba 0.5469 0 1 1>}
  n10 = GetMaterialAttributes(None=NamedRerouteUsage()) {attributeGetTypes [
    #1
    #2
    #3
    #4
   ]}
  NamedRerouteDeclaration(None=@n10.Normal) {name=AsphaltNormal, nodeColor=<rgba 0.5469 0 1 1>}
  NamedRerouteDeclaration(None=@n8.R) {name=AsphaltAreaMask, nodeColor=<rgba 0.5469 0 1 1>}
  n11 = Constant() {r=0.5}
  n12 = TextureObject() {texture=/Game/Environments/Sets/Arid/Textures/T_Swirly_Noise_D}
  n13 = Constant() {r=350}
  n14 = LinearInterpolate(
   A=@n11
   B=Constant() {r=0.9575}
   Alpha=ComponentMask(None=MaterialFunctionCall(TextureObject=@n12, TextureSize=@n13) {materialFunction=&WorldAlignedTexture}.Z Texture) {b=true}
  )
  n15 = NamedRerouteUsage()
  n16 = TextureSampleParameter2D!(UVs, Apply View MipBias) {
   group=Road Lines
   parameterName=Road Line Normal
   samplerSource=SSM_Wrap_WorldGroupSettings
   samplerType=SAMPLERTYPE_Normal
   texture=/Game/Athena/Apollo/Environments/Landscape/Textures/Roads/AsphaltRoad1_normal
  }
  n17 = Transform(None=Multiply(A=@n16.RGB, B=Constant3Vector() {constant=<rgba 1.7 1.7 1 0>}))
  n18 = LinearInterpolate(
   A=MaterialFunctionCall(BaseNormal=NamedRerouteUsage(), AdditionalNormal=@n17) {materialFunction=&BlendAngleCorrectedNormals}.Result
   B=@n17
   Alpha=@n15
  )
  n19 = NamedRerouteUsage()
  n20 = NamedRerouteUsage()
  n21 = VertexNormalWS()
  n22 = SetMaterialAttributes!(
   MaterialAttributes=@n10.MaterialAttributes
   Roughness=LinearInterpolate(A=@n10.Roughness, B=ShadingPathSwitch(Default=@n14, Forward=@n11, Mobile=@n11), Alpha=@n15)
   Specular=LinearInterpolate(A=@n10.Specular, B=Constant() {r=0.6}, Alpha=@n15) {constB=0.6}
   BaseColor=LinearInterpolate(A=@n10.BaseColor, B=NamedRerouteUsage(), Alpha=@n15)
   Normal=ShadingPathSwitch(Default=LinearInterpolate(A=@n18, B=@n19, Alpha=@n20), Forward=@n21, Mobile=@n21)
  ) {attributeSetTypes [
    #1
    #2
    #3
    #4
   ]}
  NamedRerouteDeclaration(None=@n22) {name=RoadLinesOnAsphalt, nodeColor=<rgba 0.5469 0 1 1>}
  ScalarParameter() {defaultValue=-0.5, parameterName=MIpBias}
  ScalarParameter() {defaultValue=0.9575, group=Road, parameterName=Lines_R}
  ScalarParameter() {defaultValue=0.5, group=Road, parameterName=Lines_R2}
  n23 = GetMaterialAttributes(None=NamedRerouteUsage()) {attributeGetTypes [#3, #1]}
  n24 = VectorParameter() {group=Road, parameterName=RoadTint_Edge, defaultValue=<rgba 0.5625 0.5625 0.5625 1>}
  n25 = VectorParameter() {group=Road, parameterName=RoadTint_Center, defaultValue=<rgba 0.151 0.151 0.151 0>}
  n26 = MaterialFunctionCall!(Mask Width Push, Gradient Mask Value) {materialFunction=/Game/Environments/Sets/Arid/Materials/MF_RoadMaskGradient}
  SetMaterialAttributes!(
   MaterialAttributes=@n23.MaterialAttributes
   BaseColor=Multiply(A=@n23.BaseColor, B=LinearInterpolate(A=@n24.RGB, B=@n25.RGB, Alpha=@n26.Result))
   Roughness=LinearInterpolate(A=Multiply(B=@n23.Roughness) {constA=1}, B=@n23.Roughness, Alpha=@n26.Result)
  ) {attributeSetTypes [#3, #1]}
  VectorParameter() {group=Road Lines, parameterName=RoadTile N Value, defaultValue=<rgba 1.7 1.7 1 0>}
 ]
}
