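# Material function that computes aggregate blend alpha, tire mask and V-axis fade mask for Apollo landscape roads.
# World-space grayscale noise and rerouted masks feed a HeightLerp for aggregate breakup, bypassed on Forward and Mobile shading paths.
# Static switches toggle aggregate, vertex-color or UV-gradient edge masking, fade direction, and a second crack-height road fade.
# Tire Mask adds inverted alpha to Tire Alpha, or passes it through with virtual texturing.
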MaterialFunction /Game/Athena/Apollo/Environments/Landscape/Materials/MaterialFunctions/MF_ApolloRoadAggregateBlend
graph {nodes [
  n1 = ComponentMask(None=TextureCoordinate()) {g=true}
  n2 = OneMinus(None=@n1)
  n3 = Saturate(None=Add(A=Constant() {r=-0.1}, B=StaticSwitchParameter(True=@n2, False=@n1) {defaultValue=true, group=Aggregate Blend, parameterName=Fade From Top}))
  n4 = NamedRerouteUsage()
  n5 = LinearInterpolate(A=Constant() {r=5}, B=Constant() {r=-1.25}, Alpha=@n4)
  n6 = WorldPosition() {worldPositionShaderOffset=WPT_ExcludeAllShaderOffsets}
  n7 = TextureSample(UVs=Divide(A=ComponentMask(None=@n6.XYZ) {g=true, r=true}, B=Constant() {r=1650}) {constB=8192}) {
   samplerSource=SSM_Wrap_WorldGroupSettings
   samplerType=SAMPLERTYPE_Grayscale
   texture=/Game/Athena/Apollo/Environments/Landscape/Materials/Biome/CommonMaterials/T_ApolloGrass_ColorMask
  }
  n8 = Saturate(None=Add(A=@n7.R, B=Constant() {r=0.4}))
  n9 = Saturate(None=Multiply(A=OneMinus(None=@n4), B=Constant() {r=1.25}) {constB=3})
  n10 = Constant() {r=0.5}
  n11 = Multiply(
   A=Multiply(
    A=Saturate(None=@n5)
    B=MaterialFunctionCall(Transition Phase=@n8, Height Texture=@n9, Contrast=@n10) {materialFunction=&HeightLerp}.Alpha
   )
   B=NamedRerouteUsage()
  )
  n12 = NamedRerouteUsage()
  n13 = NamedRerouteUsage()
  n14 = StaticSwitchParameter(True=ShadingPathSwitch(Default=Max(A=@n11, B=@n12), Forward=@n13, Mobile=@n13), False=NamedRerouteUsage()) {defaultValue=true, group=Aggregate Blend, parameterName=Aggregate ON}
  n15 = MaterialFunctionCall(
   Transition Phase=Multiply(A=@n3, B=@n14)
   Height Texture=FunctionInput(Preview=Constant() {r=1}) {inputName=CrackHeight, inputType=FunctionInput_Scalar}
  ) {materialFunction=&HeightLerp}
  n16 = StaticSwitchParameter(True=Multiply(A=@n15.Lerp Alpha No Contrast, B=@n15.Lerp Alpha No Contrast), False=@n14) {group=Aggregate Blend, parameterName=Fade Road}
  FunctionOutput(None=@n16) {outputName=Alpha}
  n17 = FunctionInput(Preview=Constant() {r=1}) {inputName=Tire Alpha, inputType=FunctionInput_Scalar}
  n18 = Add(A=OneMinus(None=@n16), B=@n17)
  FunctionOutput(None=VirtualTextureFeatureSwitch!(No=Saturate(None=@n18), Yes=@n17)) {outputName=Tire Mask}
  FunctionOutput(None=@n3) {outputName=FadeMask, sortPriority=-1}
  NamedRerouteDeclaration(None=FunctionInput(Preview=Constant() {r=1}) {bUsePreviewValueAsDefault=true, inputName=Aggregate Alpha, inputType=FunctionInput_Scalar}) {name=AggregateAlpha, nodeColor=<rgba 0.5469 0 1 1>}
  n19 = TextureCoordinate()
  n20 = Subtract(A=Multiply(A=ComponentMask(None=@n19) {r=true}) {constB=2})
  n21 = Multiply(A=OneMinus(None=Abs(None=@n20)), B=Constant() {r=2}) {constB=5}
  n22 = Constant() {r=0.8}
  n23 = OneMinus(None=Subtract(B=Power(Base=@n21, Exp=@n22)) {constB=4})
  NamedRerouteDeclaration(None=StaticSwitchParameter(True=VertexColor().R, False=Saturate(None=@n23)) {group=Aggregate Blend, parameterName=Vert Color For Aggregate}) {name=AggregateEdgeMask, nodeColor=<rgba 0.5469 0 1 1>}
  n24 = FunctionInput(Preview=Constant() {r=1}) {bUsePreviewValueAsDefault=true, desc=Mask for where the asphalt is, inputName=Asphalt Alpha, inputType=FunctionInput_Scalar}
  NamedRerouteDeclaration(None=@n24) {name=AsphaltMask, nodeColor=<rgba 0.5469 0 1 1>}
  n25 = Multiply(A=NamedRerouteUsage(), B=NamedRerouteUsage())
  Saturate(None=Add(A=Multiply(A=@n25, B=@n25), B=@n25))
  ScalarParameter() {defaultValue=1650, group=Aggregate, parameterName=Z-Aggregate_NoiseScale}
  ScalarParameter() {defaultValue=5, group=Aggregate, parameterName=Z-Aggregate_HeightL-A}
  ScalarParameter() {defaultValue=-1.25, group=Aggregate, parameterName=Z-Aggregate_HeightL-B}
  ScalarParameter() {defaultValue=0.5, group=Aggregate, parameterName=Z-Aggregate_HeightCont}
  ScalarParameter() {defaultValue=1.25, group=Aggregate, parameterName=Z-Aggregate_HeightMod}
  ScalarParameter() {defaultValue=0.8, parameterName=Gradient Exponent, sliderMax=4, sliderMin=0.5}
  ScalarParameter() {defaultValue=0.4, group=Aggregate, parameterName=Z-Aggregate_NoiseAdd}
  ScalarParameter() {defaultValue=2, parameterName=Gradient offset, sliderMax=10, sliderMin=0.1}
 ]}
