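# Material function for the asphalt aggregate surface on Apollo landscape roads, outputting material attributes plus an Alpha from the spec mask's blue channel.
# World-aligned Z projection samples asphalt detail textures, brightening diffuse by division, deriving specular and roughness from mask channels, and strengthening the normal.
# A shading path switch gives forward and mobile the same base color with constant specular and roughness and vertex normals.
# Aggregate parameter nodes remain unconnected, with constants used instead.
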MaterialFunction /Game/Athena/Apollo/Environments/Landscape/Materials/MaterialFunctions/MF_ApolloRoad_Aggregate
graph {nodes [
  n1 = MaterialFunctionCall!(
   Diffuse Texture Object=TextureObject() {texture=/Game/Athena/Apollo/Environments/Landscape/Textures/Roads/T_AsphaltDetail1_D}
   Spec Texture Object=TextureObject() {samplerType=SAMPLERTYPE_Masks, texture=/Game/Athena/Apollo/Environments/Landscape/Textures/Roads/T_AsphaltDetail1_S}
   TextureSize=Constant() {r=600}
   --------------------------
   Use seperate settings for Z projection
   Use seperate texture for Z projection
   Texture Size Z projection
   Z Plane Diffuse Texture Object
   Z Plane Spec Texture Object
   "--------------------------      "
   ProjectionTransitionContrast
   Use Texture Alpha Heightmap to Refine the Transistion
   HeightLerpTransitionContrast
   PickTheDiffuseHeightChannel
   TheDiffuseHeightChannel
   "--------------------------       "
   NormalTextureObject=TextureObject() {samplerType=SAMPLERTYPE_Normal, texture=/Game/Athena/Apollo/Environments/Landscape/Textures/Roads/T_AsphaltDetail1_N}
   World Space Normals=StaticBool() {value=true}
   "--------------------------  "
   Export Float 4 Diffuse=StaticBool()
   Export Float 4 Spec
   "--------------------------     "
   WorldPosition=WorldPosition() {worldPositionShaderOffset=WPT_ExcludeAllShaderOffsets}.XYZ
   World Space Normal
  ) {materialFunction=/Engine/Functions/Engine_MaterialFunctions02/Texturing/WorldAlignedTextures_Complex}
  n2 = Constant() {r=0.2}
  n3 = ComponentMask(None=@n1.Z Spec) {b=true}
  n4 = Constant3Vector() {constant=<rgba 4 4 1 1>}
  n5 = Saturate(None=Divide(A=@n1.Z Diffuse Texture, B=Constant() {r=0.7}))
  n6 = SetMaterialAttributes!(
   MaterialAttributes
   BaseColor=@n5
   Specular=Multiply(A=Constant() {r=0.5}, B=ComponentMask(None=@n1.Z Spec) {r=true})
   Roughness=Saturate(None=Add(A=@n2, B=@n3))
   Normal=Normalize(VectorInput=Multiply(A=@n1.Normal Z Texture, B=@n4))
  ) {attributeSetTypes [
    #1
    #2
    #3
    #4
   ]}
  n7 = SetMaterialAttributes(BaseColor=@n5, Specular=Constant() {r=0.3}, Roughness=Constant() {r=0.9}, Normal=VertexNormalWS()) {attributeSetTypes [
    #1
    #2
    #3
    #4
   ]}
  FunctionOutput(None=ShadingPathSwitch(Default=@n6, Deferred=@n6, Forward=@n7, Mobile=@n7))
  FunctionOutput(None=@n3) {outputName=Alpha, sortPriority=1}
  ScalarParameter() {defaultValue=0.2, group=Aggregate, parameterName=Aggregate_RoughAdd}
  ScalarParameter() {defaultValue=620, group=Aggregate, parameterName=Aggregate_TilingSize2}
  ScalarParameter() {defaultValue=0.5, group=Aggregate, parameterName=Aggregate_SpecMod-A}
  VectorParameter() {parameterName=Aggregate_ColorDiv, defaultValue=<rgba 0.7 0.7 0.7 1>}
  VectorParameter() {parameterName=Aggregate_NormAmt, defaultValue=<rgba 4 4 1 1>}
 ]}
