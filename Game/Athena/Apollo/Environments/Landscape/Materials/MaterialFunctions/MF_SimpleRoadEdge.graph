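# Material function for a simple road edge in Apollo landscape materials, outputting material attributes plus edge alpha, height and Runtime Virtual Texture blend masks.
# Tinted, desaturated diffuse, normal and SMRA textures build the surface; metallic, specular and roughness all read SMRA green.
# Switchable parallax occlusion mapping offsets pixel-shader UVs, with step counts reduced through a depth-based optimization mask.
# World-position noise subtracts from the RVT blend mask and is also output.
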MaterialFunction /Game/Athena/Apollo/Environments/Landscape/Materials/MaterialFunctions/MF_SimpleRoadEdge
graph {nodes [
  n1 = TextureObjectParameter() {
   group=SimpleRoad
   parameterName=SMRH POM Tex
   texture=/Game/Athena/Apollo/Environments/Landscape/Materials/MaterialFunctions/ExampleTextures/T_ExampleSimpleRoad_SMR
  }
  n2 = ScalarParameter() {defaultValue=0.02, group=SimpleRoad, parameterName=POM Height Ratio}
  n3 = ScalarParameter() {defaultValue=2, group=SimpleRoad, parameterName=POM Min Steps}
  n4 = NamedRerouteUsage()
  n5 = ScalarParameter() {defaultValue=8, group=SimpleRoad, parameterName=POM Max Steps}
  n6 = VectorParameter() {parameterName=POMChannel, defaultValue=<rgba 0 0 0 1>}
  n7 = TextureCoordinate()
  n8 = StaticSwitchParameter(
   True=MaterialFunctionCall!(
    Heightmap Texture=@n1
    Height Ratio=@n2
    Min Steps=LinearInterpolate(A=@n3, Alpha=@n4) {constB=2}
    Max Steps=LinearInterpolate(A=@n5, Alpha=@n4) {constB=2}
    UVs
    Heightmap Channel=AppendVector(A=@n6.RGB, B=@n6.RGB)
    Reference Plane=Constant()
    -----------------------------
    Use World Coordinates
    Specify Manual Texture Size
    Manual Texture Size
    ------------------------------
    "Render Shadows (Occlusion Mapping)"
    Light Vector
    Shadow Steps
    Shadow Penumbra
    Transform To VertexNormal
   ) {materialFunction=/Engine/Functions/Engine_MaterialFunctions01/Texturing/ParallaxOcclusionMapping}.Parallax UVs
   False=@n7
  ) {group=SimpleRoad, parameterName=POM Switch}
  n9 = TextureSampleParameter2D(UVs=ShaderStageSwitch(PixelShader=@n8, VertexShader=@n7)) {
   group=SimpleRoad
   parameterName=SimpleRoadMasks
   texture=/Game/Athena/Apollo/Environments/Landscape/Materials/MaterialFunctions/ExampleTextures/T_ExampleSimpleRoad_M
   channelNames {a=Height, b="", g=EdgeBlend, r=RVTblend}
  }
  n10 = Noise!(
   World Position=Multiply(A=WorldPosition().XYZ, B=ScalarParameter() {defaultValue=0.5, group=SimpleRoad, parameterName="RVT Noise Scale (R)"})
   FilterWidth
  ) {scale=0.005}
  n11 = LinearInterpolate(
   A=ScalarParameter() {defaultValue=0.104, group=SimpleRoad, parameterName=RVT Noise Contrast LOW}
   B=ScalarParameter() {defaultValue=0.648, group=SimpleRoad, parameterName=RVT Noise Contrast High}
   Alpha=@n10
  )
  n12 = Saturate(None=Multiply(A=Saturate(None=@n11), B=ScalarParameter() {defaultValue=0.8, group=SimpleRoad, parameterName=RVT_Noise_Opacity}))
  Add(A=@n9.RVTblend, B=@n12)
  n13 = VectorParameter() {
   group=SimpleRoad
   parameterName=SimpleRoadDiffuse_Tint
   channelNames {a=SimpleRoadDiffuse_Desaturation, b="", g="", r=""}
   defaultValue=<rgba 1 1 1 0>
  }
  n14 = TextureSampleParameter2D(UVs=@n8) {
   group=SimpleRoad
   parameterName=SimpleRoadDiffuse
   texture=/Game/Athena/Apollo/Environments/Landscape/Materials/MaterialFunctions/ExampleTextures/T_ExampleSimpleRoad
  }
  n15 = Desaturation(None=@n14.RGB, Fraction=@n13.SimpleRoadDiffuse_Desaturation)
  n16 = TextureSampleParameter2D(UVs=@n8) {
   group=SimpleRoad
   parameterName=SimpleRoadSMRA
   texture=&T_ExampleSimpleRoad_SMR
  }
  n17 = TextureSampleParameter2D(UVs=@n8) {
   group=SimpleRoad
   parameterName=SimpleRoadNormal
   samplerType=SAMPLERTYPE_Normal
   texture=/Game/Athena/Apollo/Environments/Landscape/Materials/MaterialFunctions/ExampleTextures/T_ExampleSimpleRoad_N
  }
  n18 = SetMaterialAttributes!(
   MaterialAttributes
   BaseColor=Saturate(None=Multiply(A=@n13, B=@n15))
   Metallic=@n16.G
   Specular=@n16.G
   Roughness=@n16.G
   Normal=Transform(None=@n17.RGB)
  ) {attributeSetTypes [
    #1
    #2
    #3
    #4
    #5
   ]}
  FunctionOutput(None=@n18) {outputName=Material}
  FunctionOutput(None=@n9.EdgeBlend) {outputName=SimpleRoadAlpha, sortPriority=1}
  FunctionOutput(None=@n9.Height) {outputName=SimpleRoadHeight, sortPriority=2}
  FunctionOutput(None=Saturate(None=Subtract(A=@n9.RVTblend, B=@n12))) {outputName=SimpleRoadRVTBlend, sortPriority=2}
  FunctionOutput(None=@n12) {outputName=RVTNoiseMask, sortPriority=200}
  Multiply(
   A=@n2
   B=LinearInterpolate(Alpha=MaterialFunctionCall() {materialFunction=&DitherTemporalAA}.Result) {constA=1}
  )
  n19 = NamedRerouteUsage()
  n20 = ScalarParameter() {defaultValue=1000, group=SimpleRoad, parameterName=POMOptimizationDistance}
  n21 = Multiply(B=@n20) {constA=0.5}
  n22 = Subtract(A=Add(A=PixelDepth().R, B=Multiply(A=@n19, B=@n21)), B=@n20)
  NamedRerouteDeclaration(None=Saturate(None=Divide(A=@n22) {constB=1000})) {name=POMOptimizationMask, nodeColor=<rgba 1 0 0.3516 1>}
  Power(Base=@n10, Exp=ScalarParameter() {defaultValue=5, group=SimpleRoad, parameterName=RVT Noise Power})
 ]}
