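# Material function for simple tiling roads in the Apollo landscape, outputting material attributes from gravel-default diffuse, mask and normal texture parameters.
# WorldAlignedTextures_Complex projects the textures in world space, with only the Z-plane results used and a tiling size input.
# An optional static switch desaturates the diffuse and blends two tints using its adjusted blue channel.
# The mask's red channel drives both Specular and Roughness.
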MaterialFunction /Game/Athena/Apollo/Environments/Landscape/Materials/MaterialFunctions/MF_SimpleRoadTile
graph {nodes [
  n1 = TextureObjectParameter() {
   group=SimpleRoad
   parameterName=TilingDiffuse
   texture=/Game/Athena/Apollo/Environments/BuildingActors/Bridges/Walls/Textures/T_RoadsideGravel_D
  }
  n2 = TextureObjectParameter() {
   group=SimpleRoad
   parameterName=Tiling_SMR
   samplerType=SAMPLERTYPE_Masks
   texture=/Game/Athena/Apollo/Environments/BuildingActors/Bridges/Walls/Textures/T_RoadsideGravel_S
  }
  n3 = StaticBool()
  n4 = TextureObjectParameter() {
   group=SimpleRoad
   parameterName=TilingNormal
   samplerType=SAMPLERTYPE_Normal
   texture=/Game/Athena/Apollo/Environments/BuildingActors/Bridges/Walls/Textures/T_RoadsideGravel_N
  }
  n5 = MaterialFunctionCall!(
   Diffuse Texture Object=@n1
   Spec Texture Object=@n2
   TextureSize=NamedRerouteUsage()
   --------------------------
   Use seperate settings for Z projection
   Use seperate texture for Z projection
   Texture Size Z projection
   Z Plane Diffuse Texture Object
   Z Plane Spec Texture Object
   "--------------------------      "
   ProjectionTransitionContrast
   Use Texture Alpha Heightmap to Refine the Transistion=@n3
   HeightLerpTransitionContrast
   PickTheDiffuseHeightChannel
   TheDiffuseHeightChannel
   "--------------------------       "
   NormalTextureObject=@n4
   World Space Normals=StaticBool() {value=true}
   "--------------------------  "
   Export Float 4 Diffuse=@n3
   Export Float 4 Spec
   "--------------------------     "
   WorldPosition=WorldPosition() {worldPositionShaderOffset=WPT_ExcludeAllShaderOffsets}.XYZ
   World Space Normal
  ) {materialFunction=/Engine/Functions/Engine_MaterialFunctions02/Texturing/WorldAlignedTextures_Complex}
  n6 = VectorParameter() {
   group=SimpleRoad
   parameterName=TilingDiffuse_Tint
   channelNames {a=TilingDiffuse_Desaturation, b="", g="", r=""}
   defaultValue=<rgba 1 1 1 0>
  }
  n7 = Desaturation(None=@n5.Z Diffuse Texture, Fraction=@n6.TilingDiffuse_Desaturation)
  n8 = VectorParameter() {
   group=SimpleRoad
   parameterName=TilingDiffuse_Tint2
   channelNames {a=TilingDiffuse_Desaturation, b="", g="", r=""}
   defaultValue=<rgba 1 1 1 0>
  }
  n9 = Saturate(None=Multiply(A=@n7, B=@n8))
  n10 = Saturate(None=Multiply(A=@n6, B=@n7))
  n11 = ComponentMask(None=@n5.Z Diffuse Texture) {b=true}
  n12 = ScalarParameter() {defaultValue=1, group=SimpleRoad, parameterName=TilingDiffuse_Tint_BluePow}
  n13 = Saturate(None=Multiply(A=Power(Base=@n11, Exp=@n12), B=ScalarParameter() {defaultValue=1, group=SimpleRoad, parameterName=TilingDiffuse_Tint_BlueMod}))
  n14 = MaterialFunctionCall(Float3=@n5.Z Spec) {materialFunction=&BreakOutFloat3Components}
  n15 = SetMaterialAttributes!(
   MaterialAttributes
   BaseColor=StaticSwitchParameter(True=LinearInterpolate(A=@n9, B=@n10, Alpha=@n13), False=@n5.Z Diffuse Texture) {group=SimpleRoad, parameterName=TilingDiffuse_UseAdvancedColor}
   Specular=@n14.R
   Roughness=@n14.R
   Normal=@n5.Normal Z Texture
  ) {attributeSetTypes [
    #1
    #2
    #3
    #4
   ]}
  FunctionOutput(None=NamedRerouteDeclaration(None=@n15) {name=Default-MaterialOut, nodeColor=<rgba 0.5469 0 1 1>}) {outputName=BlendedMaterial}
  NamedRerouteDeclaration(None=FunctionInput(Preview=Constant() {r=350}) {bUsePreviewValueAsDefault=true, inputName=TilingSize, inputType=FunctionInput_Scalar}) {name=AsphaltScale, nodeColor=<rgba 0.5469 0 1 1>}
 ]}
