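# Master masked material for Apollo spline-mesh roads, covering asphalt, aggregate shoulders, a simple road variant and optional snow.
# Road function layers blend by vertex color and height lerps into a reroute-linked attribute graph.
# Static switches select landscape RVT sampling or non-RVT fades, alongside RVT output, debug color and snow passes.
# The mobile output uses simplified asphalt with painted lines and camera-distance-scaled world position offset toward the viewer.
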Material /Game/Athena/Apollo/Environments/Landscape/Materials/M_Apollo_Roads_Master
graph {
 properties {
  bCastRayTracedShadows=false
  bTangentSpaceNormal=false
  bUseMaterialAttributes=true
  bUsedWithSplineMeshes=true
  blendMode=BLEND_Masked
  physMaterial=/Game/FortressPhysicalMaterials/Asphalt
  pixelDepthOffsetMode=PDOM_Legacy
  usedShadingModels=Default Lit
 }
 nodes [
  n1 = MaterialFunctionCall() {materialFunction=/Game/Athena/Apollo/Environments/Landscape/Materials/MaterialFunctions/MF_ApolloRoad_Aggregate}
  n2 = MaterialFunctionCall!(Aggregate In=@n1.Result) {materialFunction=/Game/Athena/Apollo/Environments/Landscape/Materials/MaterialFunctions/MF_ApolloRoad}
  n3 = MaterialFunctionCall!(Asphalt Alpha=@n2.Alpha, Tire Alpha=@n2.Alpha, CrackHeight=@n2.Alpha, Aggregate Alpha=@n1.Alpha) {materialFunction=/Game/Athena/Apollo/Environments/Landscape/Materials/MaterialFunctions/MF_ApolloRoadAggregateBlend}
  NamedRerouteDeclaration(None=@n3.Alpha) {name=RoadToAggregateAlpha, nodeColor=<rgba 0.5469 0 1 1>}
  n4 = MaterialFunctionCall() {materialFunction=/Game/Athena/Apollo/Environments/Landscape/Materials/MaterialFunctions/MF_SimpleRoadEdge}
  NamedRerouteDeclaration(None=@n4.Material) {name=SimpleRoadEdge, nodeColor=<rgba 0 0.9375 1 1>}
  NamedRerouteDeclaration(None=@n4.SimpleRoadAlpha) {name=SimpleRoadEdgeAlpha, nodeColor=<rgba 0 1 0.4609 1>}
  n5 = BlendMaterialAttributes(A=NamedRerouteUsage(), B=NamedRerouteUsage(), Alpha=NamedRerouteUsage())
  n6 = GetMaterialAttributes(None=@n5) {attributeGetTypes [#1]}
  n7 = MaterialFunctionCall() {materialFunction=/Game/Athena/Asteria/Environments/Landscape/Materials/MaterialFunctions/MF_SnowColorBase}
  n8 = ScalarParameter() {defaultValue=0.2, parameterName=SnowColorAdd}
  n9 = Multiply(A=@n6.Base Color, B=Saturate(None=Add(A=@n7.SnowAlbedo, B=@n8)))
  NamedRerouteDeclaration(None=StaticSwitchParameter(
    True=SetMaterialAttributes!(MaterialAttributes=@n6.MaterialAttributes, Base Color=@n9) {attributeSetTypes [#1]}
    False=@n5
   ) {parameterName=UseAsteriaSnowBlend}) {name=BlendedSimple RoadMaterials, nodeColor=<rgba 0 0.4674 1 1>}
  NamedRerouteDeclaration(None=@n4.SimpleRoadRVTBlend) {name=SimpleRoadRVTBlend, nodeColor=<rgba 0 0.3117 1 1>}
  NamedRerouteDeclaration(None=ComponentMask(None=MaterialFunctionCall!(UseLandscapeCoords) {materialFunction=/Game/Athena/Asteria/Environments/Landscape/Materials/MaterialFunctions/MF_AsteriaTerrainMasks}.BiomeMaskRGBA) {r=true}) {name=SnowMask, nodeColor=<rgba 1 0.8906 0 1>}
  n10 = MaterialFunctionCall() {materialFunction=/Game/Athena/Asteria/Environments/Landscape/Materials/MaterialFunctions/MobileLayers/MF_Mobile_Asteria_Snow}
  n11 = LinearInterpolate(A=ScalarParameter() {defaultValue=0.5, parameterName=AsteriaGravel EdgeBlendNonRVT}, Alpha=VertexColor().R)
  n12 = MaterialFunctionCall!(OverrideEdgeColor, BiomeMask-RGBA, EdgeMask=@n11, BypassBiomeColors) {materialFunction=/Game/Athena/Asteria/Environments/Landscape/Materials/MaterialFunctions/MobileLayers/MF_Mobile_Asteria_Gravel}
  n13 = GetMaterialAttributes(None=NamedRerouteUsage()) {attributeGetTypes [#1]}
  n14 = Add(
   A=MaterialFunctionCall() {materialFunction=&MF_SnowColorBase}.SnowAlbedo
   B=ScalarParameter() {defaultValue=0.2, parameterName=SnowColorAdd}
  )
  n15 = SetMaterialAttributes(MaterialAttributes=@n13.MaterialAttributes, Base Color=Multiply(A=@n13.Base Color, B=Saturate(None=@n14))) {attributeSetTypes [#1]}
  n16 = NamedRerouteUsage()
  NamedRerouteDeclaration(None=BlendMaterialAttributes(
    A=StaticSwitchParameter(True=@n10.Result, False=@n12.ForStaticMesh) {parameterName=UseAsteriaSnowBlend}
    B=StaticSwitchParameter(True=@n15, False=@n16) {parameterName=UseAsteriaSnowBlend}
    Alpha=NamedRerouteUsage()
   ) {desc="Vertex blend type altered!", vertexAttributeBlendType=UseA}) {name=Mobile_SimpleRoadMaterial, nodeColor=<rgba 0 0.4674 1 1>}
  NamedRerouteDeclaration(None=@n4.RVTNoiseMask) {name=RVTNoiseMask, nodeColor=<rgba 0.1016 0 1 1>}
  n17 = GetMaterialAttributes(None=MaterialFunctionCall!(
    LayerMask=Constant() {r=1}
    InputMaterialHeightmap=Constant() {r=1}
    UseVertexNormalsForBasicSnow
    UseSparkle
    UseTangentSnowNormals
    IsTangentSpaceInput
    Material
   ) {materialFunction=/Game/Athena/Asteria/Environments/Landscape/Materials/MaterialFunctions/MF_Asteria_Snow}.PreRVT_SnowMaterial) {attributeGetTypes [#2]}
  NamedRerouteDeclaration(None=SetMaterialAttributes!(MaterialAttributes=@n17.MaterialAttributes, Normal=Transform(None=@n17.Normal)) {attributeSetTypes [#2]}) {name=AsteriaSnow_X, nodeColor=<rgba 1 0.5625 0 1>}
  n18 = NamedRerouteUsage()
  n19 = RuntimeVirtualTextureSample!(UVs, World Position) {materialType=BaseColor_Normal_Specular_YCoCg, virtualTexture=/Game/Athena/Apollo/Maps/Landscape/VT/RVT_Apollo_Landscape}
  n20 = SetMaterialAttributes!(MaterialAttributes, Base Color=@n19.BaseColor, Specular=@n19.BaseColor, Roughness=@n19.BaseColor, Normal=@n19.BaseColor) {attributeSetTypes [
    #1
    #3
    #4
    #2
   ]}
  n21 = StaticSwitchParameter(
   True=BlendMaterialAttributes(A=NamedRerouteUsage(), B=@n20, Alpha=VertexColor().B) {desc="Vertex blend type altered!", vertexAttributeBlendType=UseA}
   False=@n20
  ) {parameterName=UseSnow BridgeBlueVertexColorBlend}
  n22 = ComponentMask(None=TextureCoordinate()) {g=true}
  n23 = StaticSwitchParameter(True=@n22, False=OneMinus(None=@n22)) {group=Non RVT to RVT Fade, parameterName=V direction Fade UP}
  n24 = StaticSwitch(True=OneMinus(None=NamedRerouteUsage()), False=Multiply(A=NamedRerouteUsage(), B=Saturate(None=@n23)), Value=NamedRerouteUsage())
  NamedRerouteDeclaration(None=StaticSwitchParameter(True=BlendMaterialAttributes(A=@n18, B=@n21, Alpha=@n24) {desc="Vertex blend type altered!", vertexAttributeBlendType=UseA}, False=@n20) {group=Non RVT to RVT Fade, parameterName=Non RVT Fade to RVT}) {name=RVT_READ_Material, nodeColor=<rgba 0.5469 0 1 1>}
  n25 = ScalarParameter() {defaultValue=350, desc=This adjusts the TilingDiffuse textures, group=SimpleRoad, parameterName=TilingDiffuse_Size}
  n26 = ComponentMask(None=TextureCoordinate()) {g=true}
  n27 = LinearInterpolate(Alpha=@n26) {constA=2, constB=0}
  n28 = LinearInterpolate(Alpha=@n26) {constB=2}
  n29 = BlendMaterialAttributes(
   A=@n2.Result
   B=MaterialFunctionCall() {materialFunction=/Game/Environments/Asteria/Sets/NeonCity/MaterialFunctions/MF_NeonCity_Asphalt}.Result
   Alpha=StaticSwitchParameter(True=Saturate(None=@n27), False=Saturate(None=@n28)) {parameterName=FlipTransition}
  )
  n30 = StaticBoolParameter() {group=SimpleRoad, parameterName=UseSimpleRoad, sortPriority=-500}
  NamedRerouteDeclaration(None=StaticSwitch(
    True=MaterialFunctionCall!(TilingSize=@n25) {materialFunction=/Game/Athena/Apollo/Environments/Landscape/Materials/MaterialFunctions/MF_SimpleRoadTile}.BlendedMaterial
    False=StaticSwitchParameter(True=@n29, False=@n2.Result) {parameterName=Fade to NeonCity Asphalt}
    Value=NamedRerouteDeclaration(None=@n30) {name=UseSimpleRoad, nodeColor=<rgba 0.5469 0 1 1>}
   )) {name=RoadMaterialX, nodeColor=<rgba 0.5469 0 1 1>}
  NamedRerouteDeclaration(None=@n2.AsphaltMaterial) {name=AsphaltMaterial, nodeColor=<rgba 0.5469 0 1 1>}
  n31 = StaticSwitchParameter(
   True=MaterialFunctionCall(EdgeMask=@n11) {materialFunction=&MF_Mobile_Asteria_Gravel}.Material
   False=MaterialFunctionCall!(OverrideEdgeColor, EdgeMask, BypassBiomeColors) {materialFunction=/Game/Athena/Apollo/Environments/Landscape/Materials/Biome/CommonMaterials/Mobile/MF_Mobile_Common_Gravel}.ForStaticMesh
  ) {parameterName=Use Asteria Landscape Materials}
  n32 = StaticSwitchParameter(
   True=BlendMaterialAttributes(A=@n31, B=NamedRerouteUsage(), Alpha=VertexColor().A) {desc="Vertex blend type altered!", vertexAttributeBlendType=UseA}
   False=@n31
  ) {group=Mobile, parameterName=Mobile Blend to Asphalt}
  n33 = NamedRerouteUsage()
  n34 = NamedRerouteUsage()
  n35 = Constant3Vector() {constant=<rgba 0 0 4 0>}
  n36 = Divide(
   A=Distance(A=CameraPositionWS(), B=WorldPosition().XYZ)
   B=ScalarParameter() {defaultValue=2500, group=Mobile, parameterName=Radius of Mobile Height Bias}
  )
  n37 = Multiply(A=@n35, B=Min(A=Floor(None=@n36)) {constB=10})
  n38 = VertexColor()
  n39 = Multiply(
   A=Multiply(A=ScalarParameter() {defaultValue=2.75, group=Mobile, parameterName=Mobile Aggregate Lower Amount}, B=OneMinus(None=@n38.R))
   B=@n35
  )
  NamedRerouteDeclaration(None=SetMaterialAttributes!(
    MaterialAttributes=BlendMaterialAttributes(A=@n32, B=@n33, Alpha=@n34) {desc="Vertex blend type altered!", vertexAttributeBlendType=UseA}
    World Position Offset=Subtract(A=@n37, B=@n39)
   ) {attributeSetTypes [#5]}) {name=Reg_RoadMaterial_Mobile, nodeColor=<rgba 0.5469 0 1 1>}
  NamedRerouteDeclaration(None=@n3.FadeMask) {name=FadeMask, nodeColor=<rgba 0.5469 0 1 1>}
  n40 = NamedRerouteUsage()
  n41 = MaterialFunctionCall!(Grass_UseCustomSlopeMask, Grass_CustomSlopeMask) {materialFunction=/Game/Athena/Helios/Environments/Landscape/Materials/MaterialFunctions_Utilities/MF_Helios_StaticMeshLandscapeLayerMatch}
  n42 = NamedRerouteUsage()
  n43 = VertexColor()
  n44 = NamedRerouteUsage()
  n45 = BlendMaterialAttributes(
   A=StaticSwitchParameter(
    True=@n41.SnowLayer
    False=MaterialFunctionCall() {materialFunction=&MF_Mobile_Common_Gravel}.Material
   ) {group=HeliosSnow, parameterName=UseHeliosSnow}
   B=ShadingPathSwitch(Default=NamedRerouteUsage(), Forward=@n42, Mobile=@n42)
   Alpha=Multiply(
    A=NamedRerouteUsage()
    B=MaterialFunctionCall(Transition Phase=@n43.A, Height Texture=@n44) {materialFunction=&HeightLerp}.Alpha
   )
  ) {desc="Vertex blend type altered!", vertexAttributeBlendType=UseA}
  n46 = Saturate(None=Multiply(
    A=ScalarParameter() {desc="Blue=Asphalt Alpha=Gravel", group=Non RVT Settings, parameterName=NonRVT-UseBlueVertColorForAsphalt, sortPriority=0}
    B=VertexColor().B
   ))
  n47 = NamedRerouteUsage()
  n48 = Multiply(
   A=StaticSwitchParameter(True=NamedRerouteUsage(), False=Constant() {r=1}) {group=Non RVT Settings, parameterName=FadeRoadToAsphalt, sortPriority=0}
   B=OneMinus(None=MaterialFunctionCall(Transition Phase=@n46, Height Texture=@n47) {materialFunction=&HeightLerp}.Alpha)
  )
  NamedRerouteDeclaration(None=StaticSwitch(
    True=NamedRerouteUsage()
    False=BlendMaterialAttributes(A=@n40, B=@n45, Alpha=@n48) {desc="Vertex blend type altered!", vertexAttributeBlendType=UseA}
    Value=NamedRerouteUsage()
   )) {name=RVT_Disabled_Path, nodeColor=<rgba 0.5469 0 1 1>}
  NamedRerouteDeclaration(None=@n2.AsphaltCrackHeight) {name=AsphaltCrackHeight, nodeColor=<rgba 0.5469 0 1 1>}
  Reroute(None=NamedRerouteUsage())
  n49 = GetMaterialAttributes(None=NamedRerouteUsage()) {attributeGetTypes [
    #1
    #3
    #4
    #2
   ]}
  n50 = NamedRerouteUsage()
  n51 = NamedRerouteUsage()
  n52 = NamedRerouteUsage()
  RuntimeVirtualTextureOutput!(
   BaseColor=@n49.Base Color
   Specular=@n49.Base Color
   Roughness=@n49.Base Color
   Normal=@n49.Base Color
   WorldHeight
   Opacity=ShadingPathSwitch(Default=StaticSwitch(True=@n50, False=@n51, Value=@n52), Mobile=Constant())
   Mask
   Displacement
   Mask4
  )
  n53 = TextureCoordinate()
  n54 = Frac(None=MaterialFunctionCall(X=ComponentMask(None=@n53) {r=true}, Y=ComponentMask(None=@n53) {g=true}) {materialFunction=&MakeFloat3}.Result)
  n55 = ComponentMask(None=VectorParameter() {group=Debug, parameterName=Override Color, defaultValue=<rgba 1 0 0 1>}.RGB) {b=true, g=true, r=true}
  n56 = NamedRerouteUsage()
  n57 = NamedRerouteUsage()
  n58 = StaticSwitch(
   True=ShadingPathSwitch(Default=NamedRerouteUsage(), Forward=@n56, Mobile=@n56)
   False=ShadingPathSwitch(Default=NamedRerouteUsage(), Forward=@n57, Mobile=@n57)
   Value=NamedRerouteUsage()
  )
  n59 = NamedRerouteUsage()
  n60 = NamedRerouteDeclaration(None=MaterialFunctionCall!(
    Material=StaticSwitchParameter(True=@n58, False=@n59) {defaultValue=true, group=RVT, parameterName=EnableRVT}
    OpacityMask=NamedRerouteUsage()
    RoadTypeEdgeFalloffMask
   ) {materialFunction=/Game/Athena/Apollo/Environments/Landscape/Materials/MaterialFunctions/MF_HeliosSnow}.Result) {name=BlendedMaterial, nodeColor=<rgba 0.5469 0 1 1>}
  n61 = MaterialFunctionCall(
   NoRVT=@n60
   RVT=MaterialFunctionCall!(RVT=NamedRerouteUsage(), ByPass RVT=@n60) {materialFunction=/Game/Athena/Apollo/Maps/Landscape/VT/MaterialFunctions/EnableRVT}.Result
  ) {materialFunction=&FN_RVT_Switch_Material}
  n62 = StaticSwitchParameter(
   True=SetMaterialAttributes!(
    MaterialAttributes
    Base Color=StaticSwitchParameter(True=@n54, False=@n55) {defaultValue=true, desc=Use this to show the spline mesh UV directionality, group=Debug, parameterName=Show UV's}
   ) {attributeSetTypes [#1]}
   False=StaticSwitchParameter(True=QualitySwitch(Default=@n61.Result, Low=@n60, High=@n60, Medium=@n60, Epic=@n61.Result), False=@n61.Result) {parameterName=DisableRVTonHIghandLower}
  ) {
   desc="Use this to identify where materials are used in the map\r\nand also the spline mesh UVs. Dont submit enabled!"
   group=Debug
   parameterName=Enable Debug Color
  }
  n63 = GetMaterialAttributes(None=@n62) {attributeGetTypes [#1]}
  n64 = Multiply(A=@n63.Base Color, B=ScalarParameter() {parameterName=SSS Color Brightness})
  n65 = ScalarParameter() {defaultValue=1, parameterName=SSS Opacity}
  n66 = StaticSwitchParameter(True=Multiply(A=NamedRerouteUsage(), B=NamedRerouteUsage()), False=Constant() {r=1}) {parameterName=UseAsteriaSnowBlendTransition}
  n67 = Saturate(None=Add(A=@n66, B=NamedRerouteUsage()))
  n68 = Constant()
  n69 = FeatureLevelSwitch(
   Default=BlendMaterialAttributes(
    A=SetMaterialAttributes!(MaterialAttributes=@n63.MaterialAttributes, Subsurface Color=@n64, Opacity=@n65) {attributeSetTypes [#6, #7]}
    B=MaterialFunctionCall!(Mask=@n67, SlipperySnow_Mask=@n68, "Material;"=@n62) {materialFunction=/Game/Athena/Asteria/Environments/Landscape/Materials/MaterialFunctions/MF_SnowPostRVT}.SnowMaterial
    Alpha=@n66
   )
   ES3_1=@n62
  )
  n70 = GetMaterialAttributes(None=MaterialFunctionCall!(
    LayerWeight=Constant() {r=1}
    VariationMasks Mobile=MaterialFunctionCall() {materialFunction=/Game/Athena/Helios/Environments/Landscape/Materials/MaterialFunctions_Utilities/MF_VariationMasks_Mobile}.VariationMasks Mobile
   ) {materialFunction=/Game/Athena/Helios/Environments/Landscape/Materials/QualityLevelLayers/MF_Helios_Asphalt_NS}.Result) {attributeGetTypes [#1, #2]}
  n71 = TextureSampleParameter2D!(UVs, Bias=Constant() {r=-0.7}, Apply View MipBias) {
   group=Road Lines
   mipValueMode=TMVM_MipBias
   parameterName=Road Line Mask
   samplerType=SAMPLERTYPE_Masks
   texture=/Game/Athena/Apollo/Environments/Landscape/Textures/Roads/AsphaltRoad1_rgb
  }
  n72 = LinearInterpolate(
   A=VectorParameter() {group=Road Lines, parameterName=EdgeLine, defaultValue=<rgba 0.7344 0.6848 0.6196 0>}.RGB
   B=VectorParameter() {group=Road Lines, parameterName=CenterLine, defaultValue=<rgba 0.4125 0.2346 0 1>}.RGB
   Alpha=@n71.B
  )
  n73 = Add(A=Multiply(A=@n70.Normal) {constB=2}) {constB=-1}
  n74 = Abs(None=@n73)
  n75 = Max(A=ComponentMask(None=@n74) {r=true}, B=ComponentMask(None=@n74) {g=true})
  n76 = Multiply(A=ComponentMask(None=@n70.Normal) {r=true}) {constB=4}
  n77 = Multiply(A=Saturate(None=Subtract(A=@n71.G, B=@n75)), B=Saturate(None=Subtract(A=@n76) {constB=1.75})) {constB=0.6}
  n78 = Multiply(A=OneMinus(None=@n71.R)) {constB=0.1}
  n79 = Saturate(None=Multiply(A=Step(X=@n71.R) {constY=0.2}, B=OneMinus(None=@n78)))
  n80 = DeriveNormalZ(InXY=@n73)
  n81 = ComponentMask(None=@n80) {g=true, r=true}
  n82 = TextureSampleParameter2D!(UVs, Apply View MipBias) {
   group=Road Lines
   parameterName=Road Line Normal
   samplerSource=SSM_Wrap_WorldGroupSettings
   samplerType=SAMPLERTYPE_Normal
   texture=/Game/Athena/Apollo/Environments/Landscape/Textures/Roads/AsphaltRoad1_normal
  }
  n83 = Transform(None=Multiply(A=@n82.RGB, B=Constant3Vector() {constant=<rgba 1.7 1.7 1 0>}))
  n84 = ComponentMask(None=@n83) {g=true, r=true}
  n85 = ComponentMask(None=@n80) {b=true}
  n86 = ComponentMask(None=@n83) {b=true}
  n87 = SetMaterialAttributes!(
   MaterialAttributes=@n70.MaterialAttributes
   Base Color=Multiply(A=LinearInterpolate(A=@n70.Base Color, B=@n72, Alpha=@n77), B=@n79)
   Normal=AppendVector(A=Add(A=@n81, B=@n84), B=Multiply(A=@n85, B=@n86))
  ) {attributeSetTypes [#1, #2]}
  n88 = NamedRerouteUsage()
  n89 = Constant()
  n90 = WorldPosition()
  n91 = CameraPositionWS()
  n92 = Multiply(A=Divide(A=Distance(A=@n90.XYZ, B=@n91)) {constB=700}, B=Normalize(VectorInput=Subtract(A=@n91, B=@n90.XYZ)))
  n93 = Constant3Vector() {constant=<rgba 0 0 -60 0>}
  ShadingPathSwitch(
   Default=SetMaterialAttributes!(
    MaterialAttributes=StaticSwitchParameter(True=@n69, False=@n62) {desc=Enable Subsurface Shading Model On Instance to make this work, parameterName=UseAsteriaSnowPost}
    Opacity Mask=Constant() {r=1}
   ) {attributeSetTypes [#8]}
   Mobile=SetMaterialAttributes!(
    MaterialAttributes=MaterialFunctionCall(Material=@n87, OpacityMask=@n88) {materialFunction=&MF_HeliosSnow}.Result
    Opacity Mask=ShaderStageSwitch(PixelShader=@n79, VertexShader=@n89)
    World Position Offset=ShadowReplace(Default=@n92, Shadow=@n93)
   ) {attributeSetTypes [#8, #5]}
  )
  StaticSwitchParameter(True=BlendMaterialAttributes(B=NamedRerouteUsage(), Alpha=VertexColor().A) {desc="Vertex blend type altered!", vertexAttributeBlendType=UseA}) {group=Mobile, parameterName=Mobile Blend to Asphalt}
  n94 = NamedRerouteUsage()
  n95 = NamedRerouteUsage()
  n96 = NamedRerouteUsage()
  StaticSwitchParameter(
   True=StaticSwitchParameter(True=BlendMaterialAttributes(A=@n94, B=@n95, Alpha=@n96), False=@n95) {parameterName=UseAsteriaSnowBlendTransition}
   False=@n94
  ) {parameterName=UseAsteriaSnowBlend}
 ]
}
