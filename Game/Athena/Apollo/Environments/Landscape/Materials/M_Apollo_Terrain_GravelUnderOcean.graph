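# Apollo landscape terrain material for under-ocean gravel, built on shared Helios sand layer functions with material attributes and Nanite usage.
# The default shading path takes the UnderOceanSandMaterial output fed by the QuadB terrain mask, while mobile uses the Sand_NS function with a constant up-facing world-space normal.
# A UseCenterHole static switch can zero Opacity Mask within a parameterized XY center and radius, using world-position distance.
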Material /Game/Athena/Apollo/Environments/Landscape/Materials/M_Apollo_Terrain_GravelUnderOcean
graph {
 properties {
  bTangentSpaceNormal=false
  bUseMaterialAttributes=true
  bUsedWithNanite=true
  pixelDepthOffsetMode=PDOM_Legacy
  usedShadingModels=Default Lit
 }
 nodes [
  n1 = ShadingPathSwitch(
   Default=MaterialFunctionCall!(
    TerrainMasks_QuadA
    TerrainMasks_QuadB=MaterialFunctionCall!(UseLandscapeCoords, "UseProjection(translucent)") {materialFunction=/Game/Athena/Helios/Environments/Landscape/Materials/MaterialFunctions_Utilities/MF_TerrainMasks}.QuadB_RGBA
    TerrainMasks_QuadC
    TerrainMasks_QuadD
    UseTangentNormals
    OD_MD_Normals
    OD_MD_SedimentLayers
    UseOceanLevel
   ) {materialFunction=/Game/Athena/Helios/Environments/Landscape/Materials/QualityLevelLayers/MF_Helios_Sand_Default}.UnderOceanSandMaterial
   Mobile=SetMaterialAttributes!(
    MaterialAttributes=MaterialFunctionCall!(TerrainMasks_QuadA, TerrainMasks_QuadB, TerrainMasks_QuadC, TerrainMasks_QuadD, VariationMasks Mobile) {materialFunction=/Game/Athena/Helios/Environments/Landscape/Materials/QualityLevelLayers/MF_Helios_Sand_NS}.Result
    Normal=Constant3Vector() {constant=<rgba 0 0 1 0>}
   ) {attributeSetTypes [#1]}
  )
  n2 = WorldPosition()
  n3 = VectorParameter() {parameterName=HoleCenterAndRadius, channelNames {a=Radius, b="", g=CenterY, r=CenterX}, defaultValue=<rgba 0 0 0 25000>}
  n4 = Length(None=Subtract(A=ComponentMask(None=@n2.XYZ) {g=true, r=true}, B=ComponentMask(None=@n3) {g=true, r=true})) {is=class MaterialExpressionLength}
  n5 = Constant() {r=1}
  n6 = Constant()
  StaticSwitchParameter(
   True=SetMaterialAttributes!(MaterialAttributes=@n1, Opacity Mask=If(A=@n4, B=@n3.Radius, "A > B"=@n5, "A < B"=@n6)) {attributeSetTypes [#2]}
   False=@n1
  ) {parameterName=UseCenterHole}
  Subtract()
 ]
}
