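# Master road and landscape-set material for Apollo, usable on spline meshes, with an asphalt physical material.
# World-aligned UVs drive a triplanar blend, with an optional masked top layer blended over it.
# A static switch applies either DesertRoad or SedimentGradient; the grass/base-tiling switch has identical inputs on both branches.
# An optional twist debug view writes rotation values to emissive before MF_CreativeMoveTool processes the attributes.
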Material /Game/Athena/Apollo/Environments/Landscape/Materials/M_Roads_Master
graph {
 properties {
  bUseMaterialAttributes=true
  bUsedWithSplineMeshes=true
  physMaterial=/Game/FortressPhysicalMaterials/Asphalt
  pixelDepthOffsetMode=PDOM_Legacy
  refractionMethod=RM_IndexOfRefraction
  usedShadingModels=Default Lit
 }
 nodes [
  n1 = ScalarParameter() {defaultValue=1000, group=MasterControl, parameterName=Texturesize_XY, sortPriority=1}
  n2 = MaterialFunctionCall!(
   WorldPosition
   TextureSize=AppendVector(A=AppendVector(A=@n1, B=@n1), B=ScalarParameter() {defaultValue=1350.7627, group=MasterControl, parameterName=Texturesize_Z, sortPriority=1})
  ) {materialFunction=/Game/Environments/Sets/Arid/Materials/WorldAlignedUVs}
  n3 = MaterialFunctionCall!(UVX=@n2.X, UVY=@n2.X, UVZ=@n2.X) {materialFunction=/Game/Environments/Sets/Arid/Materials/TriplanarWorldMappedBlend}
  n4 = MaterialFunctionCall!(UV) {materialFunction=/Game/Environments/Sets/Arid/Materials/MF_MaskedTopLayer}
  n5 = StaticSwitchParameter(
   True=MaterialFunctionCall!(Base Material=@n3.Result, Top Material=@n4.Result, Alpha=@n4.Result) {materialFunction=/Engine/Functions/MaterialLayerFunctions/MatLayerBlend_Standard}.Blended Material
   False=@n3.Result
  ) {parameterName=UseMaskedLayerBlend, sortPriority=1}
  n6 = StaticSwitchParameter(
   True=MaterialFunctionCall!(In=@n5) {materialFunction=/Game/Environments/Sets/Arid/Materials/DesertRoad}.Result
   False=MaterialFunctionCall!(In=@n5) {materialFunction=/Game/Environments/Sets/Arid/Materials/SedimentGradient}.Result
  ) {defaultValue=true, parameterName=IsDesertRoad, sortPriority=1}
  n7 = StaticSwitchParameter(True=@n6, False=@n6) {defaultValue=true, group=Layers, parameterName="Use Grass(true) or a BaseTilingTexture(False)", sortPriority=1}
  MaterialFunctionCall(MA In=StaticSwitchParameter(
    True=SetMaterialAttributes(MaterialAttributes=@n7, Emissive Color=@n3.Debug Rotation Values) {attributeSetTypes [#1]}
    False=@n7
   ) {group=Twist, parameterName=Debug Twist Amount, sortPriority=1}) {materialFunction=&MF_CreativeMoveTool}
 ]
}
